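# Niagara trail effect for the DoubleDuty glider, combining speed-line sprites, a head wind-cone mesh, a tail ribbon and two horn ribbons.
# Ribbon particles persist and re-snap to skeletal mesh samples each frame, fanned out by normalized link position with scrolling material UVs.
# Forward-dot inputs, partly smoothed, drive spawn rates, ribbon length and scroll speed.
# Front-end modifiers substitute menu values, and time of day tunes speed-line and tail appearance.
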NiagaraSystem /BRCosmetics/Gadgets/Assets/VinderTech_GliderChute/Glider_DoubleDuty/FX/NS_Glider_DoubleDuty_Trail
dependencies {
 usedDataInterfaces [
  &NiagaraDataInterfaceVector2DCurve
  &NiagaraDataInterfaceCurve
  &NiagaraDataInterfaceColorCurve
  &NiagaraDataInterfaceArrayInt32
  &NiagaraDataInterfaceSkeletalMesh
 ]
 usedDynamicInputs []
 usedModules [
  &self:SystemSpawnScript.SystemScriptSource.SystemScriptGraph.NiagaraNodeAssignment_0.SetVariables_#1
  &self:SystemSpawnScript.SystemScriptSource.SystemScriptGraph.NiagaraNodeAssignment_3.SetVariables_#2
  &SystemState
  &FN_Integrator
  &self:SystemSpawnScript.SystemScriptSource.SystemScriptGraph.NiagaraNodeAssignment_2.SetVariables_#3
  &self:SystemSpawnScript.SystemScriptSource.SystemScriptGraph.NiagaraNodeAssignment_1.SetVariables_#4
 ]
}
emitters {
 NE_HeadWindCone {
  emitterUpdateScript [
   &EmitterState()
   &SpawnRate(
    SpawnRate=&Add_Float(
     A=&Lerp_Float(A=0, B=12, Alpha=&Max_Float(A:NiagaraFloat=$User.ForwardDot, B:NiagaraFloat=$User.FakeForwardDot))
     B=&FN_FrontEndFloatModifier(Gameplay=0, Front End=8)
    )
    Use Spawn Probability=false
   )
  ]
  particleSpawnScript [
   &InitializeParticle(
    Color=<rgba 1 0.7887 0.5052 0.75>
    Lifetime=0.35
    Mesh Scale=&FN_FrontEndVector3Modifier(Gameplay=<2.5 1.5 1.25>, Front End=<2.5 1.5 1.5>)
    Mesh Scale Mode=::NewEnumerator4
    Position Offset=<0 360 0>
    Position Offset Coordinate Space=::Local
    Sprite Size Mode=::NewEnumerator3
    UsePositionOffset=true
    Write Material Random=true
   )
   &InitialMeshOrientation(
    Orientation Axis=<0 1 0>
    Orientation Coordinate Space=::Local
    Orientation Vector=<1 0 0.2>
    Rotation=&FN_FrontEndVector3Modifier(Gameplay=<0 0 0>, Front End=<0 0 0>)
    Rotation Coordinate Space=::NewEnumerator2
    Use Orientation Vector=true
    Use Rotation Vector=true
   )
  ]
  particleUpdateScript [
   &SkeletalMeshLocation(
    Bone Sampling Mode=::NewEnumerator4
    Inherit Velocity=false
    MaskSpawnGroup=false
    Sampled Position Offset=&FN_FrontEndVector3Modifier(Gameplay=<0 -100 -20>, Front End=<0 -110 -40>)
    Sampled Position Offset Coordinate Space:ENiagaraMeshTransforms=::NewEnumerator2
    Skeletal Mesh=$System.SkelMesh
   )
   &ParticleState()
   &ScaleColor(Scale Alpha=&RandomRangeFloat(Minimum=0.5, Maximum=1), ScaleA=true, ScaleRGB=true)
  ]
  rendererClasses [&NiagaraMeshRendererProperties]
  renderers [{rendererClass=&NiagaraMeshRendererProperties, rendererIndex=0}]
 }
 NE_HornRibbon {
  emitterSpawnScript [
   /SetVariables(Emitter.RibbonNumber=2) {isSetParameters=true}
   /SetVariables(Emitter.ParticelPerRibbon=10) {isSetParameters=true}
  ]
  emitterUpdateScript [
   &EmitterState(
    Inactive Response=::NewEnumerator1
    Life Cycle Mode=::NewEnumerator1
    Loop Behavior=::NewEnumerator1
    Loop Duration Mode=::NewEnumerator1
    UseLoopDelay=false
   )
   &SpawnBurst_Instantaneous(
    Spawn Count=&Multiply_Int(A=$Emitter.ParticelPerRibbon, B=$Emitter.RibbonNumber)
    Spawn Time=1
    Use Spawn Probability=false
    UseLoopCountLimit=false
   )
  ]
  particleSpawnScript [
   &InitializeParticle(Lifetime=5, Ribbon Width=20, Ribbon Width Mode=::NewEnumerator1, UsePositionOffset=false, Write Material Random=true)
   &PartitionParticles(
    Number Of Partitions=$Emitter.RibbonNumber
    Particles Per Partition:NiagaraInt32=$Emitter.ParticelPerRibbon
    Partition Mode:ENiagara_PartitionMode=::NewEnumerator1
   )
   /SetVariables(Particles.NormalizedRibbonPos:NiagaraFloat=$Particles.PartitionParticles.Normalized Index In Partition) {isSetParameters=true, module=SetVariables_#5}
   /SetVariables(Particles.RibbonLinkOrder:NiagaraFloat=$Particles.NormalizedRibbonPos) {isSetParameters=true, module=SetVariables_#6}
   /SetVariables(Particles.GlowColor=&ColorFromCurve(
     ColorCurve=DataInterface() {propertyValues {
       AlphaCurve {keys [<at 0 1 RCIM_Cubic>,<at 0.6603 1>,<at 1 0>]}
       BlueCurve {keys [<at 0 1 RCIM_Cubic>,<at 0.3054 0.575>]}
       GreenCurve {keys [<at 0 1 RCIM_Cubic>,<at 0.3054 0.575>]}
       RedCurve {keys [<at 0 1 RCIM_Cubic>,<at 0.3054 0.575>]}
      }}
     CurveIndex=$Particles.NormalizedRibbonPos
    )) {isSetParameters=true, module=SetVariables_#7}
   /SetVariables() {isSetParameters=true, module=SetVariables_#8}
  ]
  particleUpdateScript [
   &ParticleState(Kill Particles When Lifetime Has Elapsed=false)
   &SkeletalMeshLocation(
    Bone / Socket Index=&SelectIntFromArray(Direct Array Index=$Particles.PartitionParticles.Partition, Int Selection Array=DataInterface() {propertyValues {IntData [1,2]}})
    MaskSpawnGroup=false
    Mesh Sampling Type=::NewEnumerator5
    Position Sampling=::NewEnumerator2
    Sampled Position Offset=&Lerp_Vector(
     A=<0 0 0>
     B=&Lerp_Vector(
      A=<-100 0 0>
      B=<-200 0 0>
      Alpha=&FN_FrontEndFloatModifier(Gameplay=$System.SmoothedForwardDot, Front End=2)
     )
     Alpha=$Particles.NormalizedRibbonPos
    )
    Skeletal Mesh=$System.SkelMesh
    Socket Sampling Mode=::NewEnumerator5
   )
   /SetVariables(Particles.RibbonFacing=&RotateVectorByQuaternion(Vector To Rotate=<0 0 1>, Quaternion=$Particles.SkeletalMeshLocation.SampledRotation)) {isSetParameters=true}
   /SetVariables(Particles.UVOffset:NiagaraFloat=&Add_Float(
     A=$Particles.UVOffset
     B=&Multiply_Float(A=$Engine.DeltaTime, B=&Lerp_Float(A=-0.1, B=-1.5, Alpha=$System.SmoothedForwardDot))
    )) {isSetParameters=true, module=SetVariables_#9}
   &Color(
    Color=&ColorFromCurve(
     ColorCurve=DataInterface() {propertyValues {
       AlphaCurve {keys [<at 0 0>,<at 0.0307 0.25 RCIM_Cubic>,<at 1 0 RCIM_Cubic>]}
       BlueCurve {keys [<at 0.1623 0.0562>]}
       GreenCurve {keys [<at 0.1623 0.106>]}
       RedCurve {keys [<at 0.1623 0.1875>]}
      }}
     CurveIndex=$Particles.NormalizedRibbonPos
    )
    Scale Color=&VectorFromFloat(Value=&Lerp_Float(A=2, B=2, Alpha=$System.SmoothedForwardDot))
   )
   /SetVariables(Particles.SpawnAlpha:NiagaraFloat=&FloatFromCurve(FloatCurve=DataInterface() {propertyValues {Curve {keys [<at 0 0>,<at 1 1>]}}}, CurveIndex=$Particles.Age)) {isSetParameters=true, module=SetVariables_#10}
   &InheritSourceMovement(
    Applied Position Scale=&VectorFromFloat(Value=&Multiply_Float(
      A=-0.5
      B=&FloatFromCurve(
       FloatCurve=DataInterface() {propertyValues {Curve {keys [
           {arriveTangent=-0.1837, interpMode=RCIM_Cubic, leaveTangent=-0.1837, tangentMode=RCTM_User, time=0.4, value=0}
           {arriveTangent=1.6438, interpMode=RCIM_Cubic, leaveTangent=1.6438, tangentMode=RCTM_User, time=1, value=1}
          ]}}}
       CurveIndex=$Particles.NormalizedRibbonPos
      )
     ))
    ApplyForce=false
    ApplyPosition=true
    ApplyPositionFalloff=false
    Source Velocity:Vector3f=&TransformVector(Vector=$Engine.Owner.Velocity, Source Space=::World, Destination Space=::Local)
   )
   &SolveForcesAndVelocity(Clamp Velocity=false, Limit Acceleration=false)
   /SetVariables(Particles.Position=$Particles.SkeletalMeshLocation.SampledPosition) {isSetParameters=true}
   &ScaleRibbonWidth(Ribbon Width Scale=&Multiply_Float(
     A=&FloatFromCurve(
      FloatCurve=DataInterface() {propertyValues {Curve {keys [
          {arriveTangent=0.6817, interpMode=RCIM_Cubic, leaveTangent=0.6817, tangentMode=RCTM_User, time=0.4, value=1.4}
          {arriveTangent=-1.0525, interpMode=RCIM_Cubic, leaveTangent=-1.0525, tangentMode=RCTM_User, time=1, value=0.8}
         ]}}}
      CurveIndex=$Particles.NormalizedRibbonPos
      Scale Curve=$Particles.SpawnAlpha
     )
     B=1
    ))
   &DynamicMaterialParameters(
    Index 0 Param 1=$Particles.UVOffset
    Index 0 Param 2=&FN_FrontEndFloatModifier(Gameplay=$System.SmoothedForwardDot, Front End=2)
    Module.Index0Param0Write=true
    Module.Index0Param1Write=true
    Module.Index0Param2Write=true
    Module.Index0Param3Write=true
   )
  ]
  rendererClasses [&NiagaraRibbonRendererProperties]
  renderers [{rendererClass=&NiagaraRibbonRendererProperties}]
 }
 NE_Speed_Lines {
  emitterSpawnScript [/SetVariables() {isSetParameters=true}]
  emitterUpdateScript [
   &EmitterState()
   /SetVariables(Emitter.SpeedLineAlpha:NiagaraFloat=&Lerp_Float(
     A=$Emitter.SpeedLineAlpha
     B=&ClampFloat(Float=$User.ForwardDot)
     Alpha=&Multiply_Float(A=$Engine.DeltaTime, B=4)
    )) {isSetParameters=true}
   &SpawnRate(
    SpawnRate=&FN_FrontEndFloatModifier(Gameplay=&Multiply_Float(A=100, B=$Emitter.SpeedLineAlpha), Front End=50)
    Use Spawn Probability=false
   ) {module=SpawnRate001}
  ]
  particleSpawnScript [
   &InitializeParticle(
    Color=&MakeLinearColorFromVectorAndFloat("Vector (RGB)"=<1 1 1>, "Float (Alpha)"=&FN_FrontEndFloatModifier(Gameplay=0.25, Front End=0.1))
    Lifetime Max=0.16
    Lifetime Min=0.1
    Lifetime Mode=::NewEnumerator1
    Sprite Size Max=<4 10>
    Sprite Size Min=<3 5>
    Sprite Size Mode=::NewEnumerator2
    UsePositionOffset=true
    Write Material Random=true
   )
   &ShapeLocation(
    Handle Radius=5
    Large Radius=&FN_FrontEndFloatModifier(Gameplay=250, Front End=200)
    MaskSpawnGroup=false
    Offset=&FN_FrontEndVector3Modifier(Gameplay=<0 -200 15>, Front End=<0 -200 15>)
    Offset Mode=::NewEnumerator0
    Override Seed=false
    Rotation Axis=<0 1 0>
    Rotation Mode=::NewEnumerator1
    Shape Primitive=::NewEnumerator3
    UseOwnerScale=false
   )
   /SetVariables(Particles.Position=&Multiply_Vector(A=Linked() {linkedVariable {name=Particles.Position, type=NiagaraPosition}}, B=<1.2 1 0.8>)) {isSetParameters=true}
   &AddVelocityFromPoint(
    Module.Use Falloff Distance=false
    Origin Offset=<0 -1000 -250>
    Velocity Strength=&FN_FrontEndFloatModifier(Gameplay=1000, Front End=0)
   )
   &AddVelocity(
    Velocity=&FN_FrontEndVector3Modifier(
     Gameplay=&Add_Vector(A=<0 300 0>, B=&Multiply_VectorByFloat(Vector=<0 0 222>, Float=$System.SmoothedForwardDot))
     Front End=<0 100 8>
    )
    Velocity Speed Scale=&FN_FrontEndFloatModifier(Gameplay=-1, Front End=1)
   )
  ]
  particleUpdateScript [
   &ParticleState()
   &ScaleSpriteSize(Scale Factor=&Vector2DFromCurve(
     Vector2Curve=DataInterface() {propertyValues {XCurve {keys [<at 0 0.5 RCIM_Cubic>,<at 0.5 1 RCIM_Cubic>]}, YCurve {keys [<at 0 0 RCIM_Cubic>,<at 0.5 1 RCIM_Cubic>]}}}
     ScaleCurve=&Vector2DFromFloat(Value=1)
    ))
   &SolveForcesAndVelocity(Clamp Velocity=false, Limit Acceleration=false)
   &ScaleSpriteSizeBySpeed(Max Scale Factor=<1 50>, Min Scale Factor=<0.01 0.05>, Velocity Threshold=1)
   &ScaleColor(
    Scale Alpha=&FloatFromCurve(
     FloatCurve=DataInterface() {propertyValues {Curve {keys [
         {arriveTangent=6.8397, interpMode=RCIM_Cubic, leaveTangent=6.8397, tangentMode=RCTM_User, time=0, value=0}
         <at 0.3 1 RCIM_Cubic>
         {arriveTangent=-2.221, interpMode=RCIM_Cubic, leaveTangent=-2.221, tangentMode=RCTM_User, time=1, value=0}
        ]}}}
     Scale Curve=&FN_FrontEndFloatModifier(
      Gameplay=&Lerp_Float(
       A=0
       B=&FNTimeOfDayScalar(Night Scale=0.2, Morning Scale=0.3, Day Scale=0.4, Evening Scale=0.3)
       Alpha=&Multiply_Float(A=$System.SystemFadeIn, B=1)
      )
      Front End=&Lerp_Float(A=0, B=0.4, Alpha=$System.SystemFadeIn)
     )
    )
    Scale RGB=&VectorFromFloat(Value=&FN_FrontEndFloatModifier(Gameplay=1, Front End=1))
    ScaleA=true
    ScaleRGB=true
   )
  ]
  rendererClasses [&NiagaraSpriteRendererProperties]
  renderers [{rendererClass=&NiagaraSpriteRendererProperties}]
 }
 NE_TailRibbon {
  emitterUpdateScript [
   &EmitterState(
    Inactive Response=::NewEnumerator1
    Life Cycle Mode=::NewEnumerator1
    Loop Behavior=::NewEnumerator1
    Loop Duration Mode=::NewEnumerator1
    UseLoopDelay=false
   )
   &SpawnBurst_Instantaneous(Spawn Count=20, Spawn Time=1, Use Spawn Probability=false, UseLoopCountLimit=false)
  ]
  particleSpawnScript [
   &InitializeParticle(
    Lifetime=5
    Ribbon Width=&FN_FrontEndFloatModifier(Gameplay=110, Front End=90)
    Ribbon Width Mode=::NewEnumerator1
    UsePositionOffset=false
    Write Material Random=true
   )
   /SetVariables(Particles.NormalizedRibbonPos:NiagaraFloat=&OneMinusFloat(Float=&ReturnNormalizedExecIndex())) {isSetParameters=true, module=SetVariables_#5}
   /SetVariables(Particles.RibbonLinkOrder:NiagaraFloat=$Particles.NormalizedRibbonPos) {isSetParameters=true, module=SetVariables_#6}
   /SetVariables(Particles.GlowColor=&ColorFromCurve(
     ColorCurve=DataInterface() {propertyValues {
       AlphaCurve {keys [<at 0 1 RCIM_Cubic>,<at 0.6603 1>,<at 1 0>]}
       BlueCurve {keys [<at 0 1 RCIM_Cubic>,<at 0.3054 0.575>]}
       GreenCurve {keys [<at 0 1 RCIM_Cubic>,<at 0.3054 0.575>]}
       RedCurve {keys [<at 0 1 RCIM_Cubic>,<at 0.3054 0.575>]}
      }}
     CurveIndex=$Particles.NormalizedRibbonPos
    )) {isSetParameters=true, module=SetVariables_#7}
   /SetVariables() {isSetParameters=true, module=SetVariables_#8}
  ]
  particleUpdateScript [
   &ParticleState(Kill Particles When Lifetime Has Elapsed=false)
   &SkeletalMeshLocation(
    MaskSpawnGroup=false
    Mesh Sampling Type=::NewEnumerator5
    Position Sampling=::NewEnumerator2
    Sampled Position Offset=&Lerp_Vector(
     A=<0 0 0>
     B=&FN_FrontEndVector3Modifier(Gameplay=&Lerp_Vector(A=<100 0 0>, B=<350 0 0>, Alpha=$System.SmoothedForwardDot), Front End=<200 0 0>)
     Alpha=$Particles.NormalizedRibbonPos
    )
    Skeletal Mesh=$System.SkelMesh
    Socket Sampling Mode=::NewEnumerator5
   )
   /SetVariables(Particles.RibbonFacing=&RotateVectorByQuaternion(Vector To Rotate=<0 1 0>, Quaternion=$Particles.SkeletalMeshLocation.SampledRotation)) {isSetParameters=true}
   /SetVariables(Particles.UVOffset:NiagaraFloat=&Add_Float(
     A=$Particles.UVOffset
     B=&Multiply_Float(A=$Engine.DeltaTime, B=&Lerp_Float(A=-0.1, B=-1.5, Alpha=$System.SmoothedForwardDot))
    )) {isSetParameters=true, module=SetVariables_#9}
   &Color(
    Color=&MultiplyLinearColor(
     Linear Color=&ColorFromCurve(
      ColorCurve=DataInterface() {propertyValues {
        AlphaCurve {keys [<at 0 0>,<at 0.1631 0.25 RCIM_Cubic>,<at 1 0 RCIM_Cubic>]}
        BlueCurve {keys [<at 0.1623 0.5833>]}
        GreenCurve {keys [<at 0.1623 0.7678>]}
        RedCurve {keys [<at 0.1623 1>]}
       }}
      CurveIndex=$Particles.NormalizedRibbonPos
     )
     Scale Factor=&MakeVector4_FromLinearColor(LinearColor=&FN_Select_Color_from_Time_Of_Day(Morning=<rgba 1 1 1 1>, Day=<rgba 1 1 1 1>, Evening=<rgba 1 1 1 1>, Night=<rgba 1 1 1 0.75>))
    )
    Scale Color=&VectorFromFloat(Value=&Lerp_Float(A=1, B=2, Alpha=$System.SmoothedForwardDot))
   )
   /SetVariables(Particles.SpawnAlpha:NiagaraFloat=&FloatFromCurve(FloatCurve=DataInterface() {propertyValues {Curve {keys [<at 0 0>,<at 1 1>]}}}, CurveIndex=$Particles.Age)) {isSetParameters=true, module=SetVariables_#10}
   &InheritSourceMovement(
    Applied Position Scale=&VectorFromFloat(Value=&Multiply_Float(
      A=-0.5
      B=&FloatFromCurve(
       FloatCurve=DataInterface() {propertyValues {Curve {keys [
           {arriveTangent=-0.0194, interpMode=RCIM_Cubic, leaveTangent=-0.0194, tangentMode=RCTM_User, time=0.2, value=0}
           {arriveTangent=1.6438, interpMode=RCIM_Cubic, leaveTangent=1.6438, tangentMode=RCTM_User, time=1, value=1}
          ]}}}
       CurveIndex=$Particles.NormalizedRibbonPos
      )
     ))
    ApplyForce=false
    ApplyPosition=true
    ApplyPositionFalloff=false
    Source Velocity:Vector3f=&TransformVector(Vector=$Engine.Owner.Velocity, Source Space=::World, Destination Space=::Local)
   )
   &SolveForcesAndVelocity(Clamp Velocity=false, Limit Acceleration=false)
   /SetVariables(Particles.Position=&Lerp_Vector(
     A=Linked() {linkedVariable {name=Particles.SkeletalMeshLocation.SampledPosition, type=NiagaraPosition}}
     B=&Lerp_Vector(
      A=Linked() {linkedVariable {name=Particles.Position, type=NiagaraPosition}}
      B=Linked() {linkedVariable {name=Particles.SkeletalMeshLocation.SampledPosition, type=NiagaraPosition}}
      Alpha=&Multiply_Float(A=$Engine.DeltaTime, B=2)
     )
     Alpha=&FloatFromCurve(FloatCurve=DataInterface() {propertyValues {Curve {keys [<at 0 0>,<at 1 0>]}}}, CurveIndex=$Particles.NormalizedRibbonPos)
    )) {isSetParameters=true}
   &ScaleRibbonWidth(Ribbon Width Scale=&Multiply_Float(
     A=&FloatFromCurve(
      FloatCurve=DataInterface() {propertyValues {Curve {keys [
          {arriveTangent=0.6817, interpMode=RCIM_Cubic, leaveTangent=0.6817, tangentMode=RCTM_User, time=0.4, value=1.4}
          {arriveTangent=-1.0525, interpMode=RCIM_Cubic, leaveTangent=-1.0525, tangentMode=RCTM_User, time=1, value=0.8}
         ]}}}
      CurveIndex=$Particles.NormalizedRibbonPos
      Scale Curve=$Particles.SpawnAlpha
     )
     B=1
    ))
   &DynamicMaterialParameters(
    Index 0 Param 1=$Particles.UVOffset
    Index 0 Param 2=$System.SmoothedForwardDot
    Module.Index0Param0Write=true
    Module.Index0Param1Write=true
    Module.Index0Param2Write=true
    Module.Index0Param3Write=true
   )
  ]
  rendererClasses [&NiagaraRibbonRendererProperties]
  renderers [{rendererClass=&NiagaraRibbonRendererProperties}]
 }
}
summary {
 systemName=NS_Glider_DoubleDuty_Trail
 emitters [
  {emitterName=NE_Speed_Lines, rendererClasses [&NiagaraSpriteRendererProperties]}
  {emitterName=NE_TailRibbon, rendererClasses [&NiagaraRibbonRendererProperties]}
  {emitterName=NE_HornRibbon, rendererClasses [&NiagaraRibbonRendererProperties]}
  {emitterName=NE_HeadWindCone, rendererClasses [&NiagaraMeshRendererProperties]}
 ]
 userVariables [
  User.bIsFrontEnd {type=NiagaraBool}
  User.bIsFrontEndPreview {type=NiagaraBool}
  User.bIsFullyDeployed {type=NiagaraBool}
  User.FakeForwardDot {type=NiagaraFloat, defaultValue=0, description=""}
  User.ForwardDot {type=NiagaraFloat, defaultValue=0, description=""}
  User.RightDot {type=NiagaraFloat, defaultValue=0, description=""}
 ]
}
